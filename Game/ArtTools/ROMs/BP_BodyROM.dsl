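; Art tool Blueprint that shows Fortnite character cosmetics on two mannequins for body ROM capture with a turntable info overlay.
; The construction script records user, date, build and cosmetic, spaces the mannequins, applies the CID, and self-assigns Python-used variables against cleanup.
; switchCosmetic copies the player's mesh and CID; BeginPlay and Tick feed the overlay.
; RenderCIDs runs BodyRoms.py through a console command; helpers read vertex counts from asset tags.
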(event EventBeginPlay
  (bind _returnvalue (UserInterface|CreateWidget "/Game/FN_PublicTurntable/WBP_Turntable_UI.WBP_Turntable_UI_C"))
  (bind _output_get (Variables|Default|SetUI _returnvalue))
  (if (Variables|Default|GetdisplayUI)
    (UserInterface|Viewport|AddtoViewport _output_get 0))
  (Class|WBPTurntableUI|SetText _output_get (Variables|Default|Getuser) (Variables|Default|Getdate) (Variables|Default|Getbuild) (Variables|Default|GetobjectList) "" "0001" ""))

(event EventTick
  (params (DeltaSeconds :type Float%20%28single-precision%29))
  (Class|WBPTurntableUI|SetFrameNumber (Variables|Default|GetUI) (Cinematics|Capture|GetCaptureFrameNumber)))

(event Custom|switchCosmetic
  (bind _returnvalue (Game|GetPlayerCharacter 0))
  (bind _returnvalue_2 (Fortnite|GetFortPlayerControllerfromActor _returnvalue))
  (bind _returnvalue_3 (Game|Player|GetPlayerPawn _returnvalue_2))
  (bind _mesh (Variables|/Script/Engine.Character|GetMesh _returnvalue_3))
  (bind _skeletalmesh (Variables|/Script/Engine.SkinnedMeshComponent|GetSkeletalMesh _mesh))
  (Development|PrintString "Switch Cosmetic")
  (Development|PrintString (Utilities|GetDisplayName _skeletalmesh))
  (Variables|Default|SetSK _skeletalmesh)
  (Variables|Default|SetCID (CosmeticLoadout|GetItemDefinitionTypefromLoadoutwithActor _returnvalue_3 "(TagName=\"LoadoutSlot.Character\")" "/Script/FortniteGame.AthenaCharacterItemDefinition"))
  (CallFunction|SetMannequinstoCID))

(fn checkVertexCount
  (for _array_element (Components|GetChildrenComponents (Variables|/Script/Engine.Actor|GetRootComponent self) false)
    (bind _asstatic_mesh_component (Utilities|Casting|CastToStaticMeshComponent _array_element))
    (bind _returnvalue (CallFunction|VertexCount (Variables|/Script/Engine.StaticMeshComponent|GetStaticMesh _asstatic_mesh_component)))
    (Development|PrintString (Utilities|String|Append "staticMesh " (Utilities|GetDisplayName (Variables|/Script/Engine.StaticMeshComponent|GetStaticMesh _asstatic_mesh_component)) " " (Utilities|String|ToString(Integer) _returnvalue)))))

(fn L_Male_skeleton
  (Default|SetSkeletalMesh (Variables|Default|GetSK_L_Male)))

(fn M_Female_skeleton
  (Default|SetSkeletalMesh (Variables|Default|GetSK_M_Female)))

(fn M_Male_skeleton
  (Default|SetSkeletalMesh (Variables|Default|GetSK_M_Male)))

(fn RenderCIDs
  (bind _self self)
  (bind _returnvalue (Utilities|GetObjectName _self))
  (bind _returnvalue_2 (Utilities|String|Append "py TechArt/BodyRoms/BodyRoms.py " _returnvalue))
  (Development|PrintString _returnvalue_2 true true "(R=0.051143,G=1.000000,B=0.087426,A=1.000000)")
  (Development|ExecuteConsoleCommand _returnvalue_2))

(fn setMannequinsToCID
  (bind _returnvalue (Cosmetic|MakeVariantChannelInfosfromVariantTagContainer :ItemDefinition "/Game/Athena/Items/Cosmetics/Characters/CID_001_Athena_Commando_F_Default.CID_001_Athena_Commando_F_Default" :CachedVariantTokens (Variables|Default|GetCached%20Variant%20Tokens)))
  (Customization|SetAthenaCharacterDefandVariants (Variables|Default|GetMannequin_L) (Variables|Default|GetCID) _returnvalue)
  (Customization|SetAthenaCharacterDefandVariants (Variables|Default|GetMannequin_R) (Variables|Default|GetCID) _returnvalue)
  (Default|SetSkeletalMesh (Variables|Default|GetSK)))

(fn setSkeletalMesh
  (params (SK :type Skeletal%20Mesh%20Object%20Reference))
  (Customization|SetSkeletalMesh (Variables|Default|GetMannequin_L) SK)
  (Customization|SetSkeletalMesh (Variables|Default|GetMannequin_R) SK))

(fn setupObjectList
  (Variables|Default|SetobjectList (Utilities|GetDisplayName (Variables|Default|GetCID))))

(fn ConstructionScript
  (bind _returnvalue (Development|GetEngineVersion))
  ;[1 1 1] Python Specific Variables, Protecting From Christmas Cleanup
  (Utilities|FlowControl|Sequence
    (:then_0
      (Variables|Default|SetMED_MaleCIDs (Variables|Default|GetMED_MaleCIDs)))
    (:then_1
      (Variables|Default|SetMED_FemaleCIDs (Variables|Default|GetMED_FemaleCIDs)))
    (:then_2
      (Variables|Default|SetLRG_MaleCIDs (Variables|Default|GetLRG_MaleCIDs)))
    (:then_3
      (Variables|Default|SetendFrame (Variables|Default|GetendFrame)))
    (:then_4
      (Variables|Default|SetxRes (Variables|Default|GetxRes)))
    (:then_5
      (Variables|Default|SetyRes (Variables|Default|GetyRes)))
    (:then_6
      (Variables|Default|Setpath (Variables|Default|Getpath)))
    (:then_7
      (if (and (Utilities|IsValid (Variables|Default|GetMannequin_L) :!class /Script/Engine.KismetSystemLibrary) (Utilities|IsValid (Variables|Default|GetMannequin_R) :!class /Script/Engine.KismetSystemLibrary))
        ;[1 1 1] UserName
        (Variables|Default|Setuser (Utilities|Platform|GetPlatformUserName))
        ;[1 1 1] Date
        (Variables|Default|Setdate (Utilities|String|ToString(Text) (Utilities|Text|AsDate(fromUTC) (Math|DateTime|UTCNow) "EDT")))
        ;[1 1 1] Build #
        (Development|PrintString _returnvalue)
        (Variables|Default|Setbuild _returnvalue)
        ;[1 1 1] Object List + vertex count
        (CallFunction|SetupObjectList)
        (if (Variables|Default|GetmirroredSeparation)
          (Transformation|SetActorRelativeLocation (Variables|Default|GetMannequin_L) (Math|Vector|MakeVector (* (Variables|Default|Getseparation) -1.0)) false true)
          (Transformation|SetActorRelativeLocation (Variables|Default|GetMannequin_R) (Math|Vector|MakeVector (Variables|Default|Getseparation)) false true)
          (else
            (Transformation|SetActorRelativeLocation (Variables|Default|GetMannequin_L) (0.0 0.0 0.0) false true)
            (Transformation|SetActorRelativeLocation (Variables|Default|GetMannequin_R) (Math|Vector|MakeVector (Variables|Default|Getseparation)) false true)))
        (CallFunction|SetMannequinstoCID)))))

(fn vertexCount
  (params (staticMesh :type Static%20Mesh%20Object%20Reference))
  (bind _returnvalue (Utilities|Name|MakeLiteralName "Vertices"))
  (bind _returnvalue_2 (AssetData|CreateAssetData staticMesh))
  (bind (_outtagvalue _returnvalue_3) (AssetData|GetTagValue _returnvalue_2 _returnvalue))
  (if _returnvalue_3
    (return (Utilities|String|StringToInteger _outtagvalue))
    (else
      (return -1))))

(fn vertexCount_SkeletalMesh
  (params (skeletalMEsh :type Skeletal%20Mesh%20Object%20Reference))
  (bind _returnvalue (Utilities|Name|MakeLiteralName "Vertices"))
  (bind _returnvalue_2 (AssetData|CreateAssetData skeletalMEsh))
  (bind (_outtagvalue _returnvalue_3) (AssetData|GetTagValue _returnvalue_2 _returnvalue))
  (if _returnvalue_3
    (return (Utilities|String|StringToInteger _outtagvalue))
    (else
      (return -1))))
